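; BP_PlayerPawn_Jethro_RapidMelon is the NPC minion pawn for RapidMelon, managing its Athena special actor component, which supplies a minimap icon.
; On BeginPlay, the server waits for playlist data to be ready, then activates the component if the Hotfix_SpecialActorEnable scalable float is true.
; On EndPlay or server-side death, the server deactivates the component to remove the icon.
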(event EventBeginPlay
  (|CallParentFunction)
  (Networking|SwitchHasAuthority
    (:Authority
      (Class|FortAsyncActionCurrentPlaylistReady|PlaylistDataReadyAsync
        (:OnReady
          (if (ScalableFloat|GetValueasBool (Variables|Default|GetHotfix_SpecialActorEnable))
            (Components|Activation|Activate (Variables|Default|GetAthenaSpecialActor) false)))))))

(event Collision|EventActorBeginOverlap
  (params (OtherActor :type Actor%20Object%20Reference))
  (|CallParentFunction OtherActor) ; This node is disabled and will not be called.\nDrag off pins to build functionality.
)

(event EventTick
  (params (DeltaSeconds :type Float%20%28single-precision%29))
  (|CallParentFunction DeltaSeconds) ; This node is disabled and will not be called.\nDrag off pins to build functionality.
)

(event EventEndPlay
  (params (EndPlayReason :type EEndPlayReason%20Enum :default "Destroyed"))
  ;[0.1719 0.06574 0] EndPlay / OnDeathServer: Deactivate special actor component (remove minimap icon)
  (|CallParentFunction EndPlayReason)
  (Networking|SwitchHasAuthority
    (:Authority
      (bind _athenaspecialactor (Variables|Default|GetAthenaSpecialActor)
        (:then
          (Components|Activation|Deactivate _athenaspecialactor))
        (:else)))))

(event EventOnDeathServer
  (params (Damage :type Float%20%28single-precision%29) (DamageTags :type Gameplay%20Tag%20Container%20Structure%20%28by%20ref%29) (Momentum :type Vector :default "0, 0, 0") (HitInfo :type Hit%20Result%20Structure%20%28by%20ref%29) (InstigatedBy :type Controller%20Object%20Reference) (DamageCauser :type Actor%20Object%20Reference) (EffectContext :type Gameplay%20Effect%20Context%20Handle%20Structure))
  ;[0.1719 0.06574 0] EndPlay / OnDeathServer: Deactivate special actor component (remove minimap icon)
  (|CallParentFunction Damage DamageTags Momentum HitInfo InstigatedBy DamageCauser EffectContext)
  (Networking|SwitchHasAuthority
    (:Authority
      (bind _athenaspecialactor (Variables|Default|GetAthenaSpecialActor)
        (:then
          (Components|Activation|Deactivate _athenaspecialactor))
        (:else)))))

(fn ConstructionScript
  (|CallParentFunction))
